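# Physics asset for the skirt of the male medium AnglePatch Royal player body, defining two dynamic cloth bone chains and lower-body collision capsules.
# Left and right chains each run four capsules from a tiny MSTR root, with tight twist limits and wider swing on one axis below the first segment.
# Free-rotation constraints link both roots to spine_01 and join spine, pelvis, thighs and calves.
# Foot capsules have no constraints.
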PhysicsAsset /BRCosmetics/Characters/Player/Male/Medium/Bodies/M_MED_AnglePatch_Royal/Meshes/M_MED_AnglePatch_Royal_Skirt_Physics
bodies {
 calf_l [{shapeType=Capsule, length=32.5, radius=10, shapeName=calf_l_capsule, center=<-18.6601 0 0>, rotation=<rot 90 0 0>}]
 calf_r [{shapeType=Capsule, length=32.5, radius=10, shapeName=calf_l_capsule, center=<18.6601 0 0>, rotation=<rot 90 0 180>}]
 dyn_clothC_l_01 [{shapeType=Capsule, length=22.5, radius=7, shapeName=dyn_clothC_l_01_capsule, center=<4.4011 -7 4>, rotation=<rot 0 0 4>}]
 dyn_clothC_l_02 [{shapeType=Capsule, length=21.5, radius=8, shapeName=dyn_clothC_l_02_capsule, center=<4.4011 -7 4>, rotation=<rot 0 0 4>}]
 dyn_clothC_l_03 [{shapeType=Capsule, length=21.5, radius=8, shapeName=dyn_clothC_l_03_capsule, center=<4.4011 -7 4>, rotation=<rot 0 0 4>}]
 dyn_clothC_l_04 [{shapeType=Capsule, length=21.5, radius=8, shapeName=dyn_clothC_l_04_capsule, center=<4.4011 -7 4>, rotation=<rot 0 0 4>}]
 dyn_clothC_l_MSTR [{shapeType=Capsule, length=0.5, radius=0.5, shapeName=dyn_clothC_l_MSTR_capsule}]
 dyn_clothC_r_01 [{shapeType=Capsule, length=22.5, radius=7, shapeName=dyn_clothC_r_01_capsule, center=<4.4011 -7 -4>, rotation=<rot 0 0 -4>}]
 dyn_clothC_r_02 [{shapeType=Capsule, length=21.5, radius=8, shapeName=dyn_clothC_l_02_capsule, center=<4.4 -7 -4>, rotation=<rot 0 0 -4>}]
 dyn_clothC_r_03 [{shapeType=Capsule, length=21.5, radius=8, shapeName=dyn_clothC_l_03_capsule, center=<4.4 -7 -4>, rotation=<rot 0 0 -4>}]
 dyn_clothC_r_04 [{shapeType=Capsule, length=21.5, radius=8, shapeName=dyn_clothC_l_04_capsule, center=<4.4 -7 -4>, rotation=<rot 0 0 -4>}]
 dyn_clothC_r_MSTR [{shapeType=Capsule, length=0.5, radius=0.5, shapeName=dyn_clothC_r_MSTR_capsule}]
 foot_l [{shapeType=Capsule, length=15.7262, radius=8.5273, shapeName=foot_l_capsule, center=<-3.4759 2.6424 0.3484>, rotation=<rot 0 0 82.4882>}]
 foot_r [{shapeType=Capsule, length=15.7262, radius=8.5273, shapeName=foot_l_capsule, center=<3.4759 -2.6424 -0.3484>, rotation=<rot 0 0 -97.5118>}]
 pelvis [{shapeType=Capsule, length=6.9594, radius=13.2385, shapeName=pelvis_capsule}]
 spine_01 [{shapeType=Capsule, length=12.5, radius=9.5, shapeName=spine_01_capsule}]
 thigh_l [{shapeType=Capsule, length=30.5, radius=13, shapeName=thigh_l_capsule, center=<-18.6601 -0.5473 -1.5455>, rotation=<rot 90 0 0>}]
 thigh_r [{shapeType=Capsule, length=26.5, radius=13, shapeName=thigh_l_capsule, center=<18.6601 0.5473 1.4895>, rotation=<rot 90 0 180>}]
}
constraints [
 {bone1Name=dyn_clothC_l_01, bone2Name=dyn_clothC_l_MSTR, swing1LimitDegrees=10, swing2LimitDegrees=10, twistLimitDegrees=5}
 {bone1Name=dyn_clothC_l_02, bone2Name=dyn_clothC_l_01, swing1LimitDegrees=30, swing2LimitDegrees=10, twistLimitDegrees=5}
 {bone1Name=dyn_clothC_l_03, bone2Name=dyn_clothC_l_02, swing1LimitDegrees=30, swing2LimitDegrees=10, twistLimitDegrees=5}
 {bone1Name=dyn_clothC_l_04, bone2Name=dyn_clothC_l_03, swing1LimitDegrees=30, swing2LimitDegrees=10, twistLimitDegrees=5}
 {bone1Name=dyn_clothC_r_01, bone2Name=dyn_clothC_r_MSTR, swing1LimitDegrees=10, swing2LimitDegrees=10, twistLimitDegrees=5}
 {bone1Name=dyn_clothC_r_02, bone2Name=dyn_clothC_r_01, swing1LimitDegrees=30, swing2LimitDegrees=10, twistLimitDegrees=5}
 {bone1Name=dyn_clothC_r_03, bone2Name=dyn_clothC_r_02, swing1LimitDegrees=30, swing2LimitDegrees=10, twistLimitDegrees=5}
 {bone1Name=dyn_clothC_r_04, bone2Name=dyn_clothC_r_03, swing1LimitDegrees=30, swing2LimitDegrees=10, twistLimitDegrees=5}
 {
  bone1Name=dyn_clothC_r_MSTR
  bone2Name=spine_01
  swing1LimitDegrees=45
  swing1Motion=Free
  swing2LimitDegrees=45
  swing2Motion=Free
  twistLimitDegrees=45
  twistMotion=Free
 }
 {
  bone1Name=dyn_clothC_l_MSTR
  bone2Name=spine_01
  swing1LimitDegrees=45
  swing1Motion=Free
  swing2LimitDegrees=45
  swing2Motion=Free
  twistLimitDegrees=45
  twistMotion=Free
 }
 {
  bone1Name=spine_01
  bone2Name=pelvis
  swing1LimitDegrees=45
  swing1Motion=Free
  swing2LimitDegrees=45
  swing2Motion=Free
  twistLimitDegrees=45
  twistMotion=Free
 }
 {
  bone1Name=thigh_l
  bone2Name=pelvis
  swing1LimitDegrees=45
  swing1Motion=Free
  swing2LimitDegrees=45
  swing2Motion=Free
  twistLimitDegrees=45
  twistMotion=Free
 }
 {
  bone1Name=calf_l
  bone2Name=thigh_l
  swing1LimitDegrees=45
  swing1Motion=Free
  swing2LimitDegrees=45
  swing2Motion=Free
  twistLimitDegrees=45
  twistMotion=Free
 }
 {
  bone1Name=thigh_r
  bone2Name=pelvis
  swing1LimitDegrees=45
  swing1Motion=Free
  swing2LimitDegrees=45
  swing2Motion=Free
  twistLimitDegrees=45
  twistMotion=Free
 }
 {
  bone1Name=calf_r
  bone2Name=thigh_r
  swing1LimitDegrees=45
  swing1Motion=Free
  swing2LimitDegrees=45
  swing2Motion=Free
  twistLimitDegrees=45
  twistMotion=Free
 }
]
