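# Post-process-style AnimBP for the AnglePatch Royal male body, adding secondary motion to skirt, cloth and loincloth bones.
# Copied parent pose feeds a crouch-aware Control Rig, cloth AnimDynamics, then skirt RigidBody.
# Loincloth uses AnimDynamics with offsets in non-emoting front end, RigidBody otherwise.
# Event-graph flags and gravity override drive these, plus corrective bone offsets for skydiving, sprinting, capes and the HightowerDate glider.
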AnimBlueprint /BRCosmetics/Characters/Player/Male/Medium/Bodies/M_MED_AnglePatch_Royal/Meshes/M_MED_AnglePatch_Royal_AnimBP
target_skeleton=&/M_MED_AnglePatch_Royal_Skeleton
graphs [
 AnimGraph {nodes [
   Output Pose(Result=Use cached pose 'PostSimCache'()) {layerGroup=None}
   n1 = Control Rig!(
    Source=Copy Pose From Mesh() {bUseAttachedParent=true}
    Alpha
    isCrouchingWalking:bool=Get isCrouchWalking() -> (isCrouchWalking:bool)
    isCrouching:double=Get isCrouching() -> (isCrouching:double)
   ) -> (Pose) {
    alphaBoolBlend {blendOption=Linear}
    controlRigAssetReference {blueprintRigClass=&/M_MED_AnglePatch_Royal_CtrlRig.M_MED_AnglePatch_Royal_CtrlRig_C}
   }
   Save cached pose 'CRCache'(Pose=@n1) {cacheName=CRCache}
   n2 = Anim Dynamics - Bone: dyn_clothA_r_02(ComponentPose=Local To Component(LocalPose=Use cached pose 'CRCache'()), Alpha=1, ExternalForce="0,0,0") {
    angularDampingOverride=1
    angularSpringConstant=8000
    bAngularSpring=true
    bOverrideAngularDamping=true
    bUseGravityOverride=true
    gravityScale=0.35
    simulationSpace=RootRelative
    alphaBoolBlend {blendOption=Linear}
    boundBone {boneName=dyn_clothA_r_02}
    componentAppliedLinearAccClamp=<2000 2000 2000>
    componentLinearAccScale=<1 1 1>
    componentLinearVelScale=<1 1 1>
    gravityOverride {z=-980}
    physicsBodyDefinitions [{
      boundBone {boneName=dyn_clothA_r_02}
      boxExtents=<20 30 30>
      constraintSetup {angularLimitsMax=<0 3 0>, angularLimitsMin=<0 -3 -15>, angularTarget=<1 0 0>}
      localJointOffset=<4 0 0>
     }]
   }
   n3 = Anim Dynamics - Bone: dyn_clothA_l_02(ComponentPose=@n2, Alpha=1, ExternalForce="0,0,0") {
    angularDampingOverride=1
    angularSpringConstant=8000
    bAngularSpring=true
    bOverrideAngularDamping=true
    bUseGravityOverride=true
    gravityScale=0.35
    simulationSpace=RootRelative
    alphaBoolBlend {blendOption=Linear}
    boundBone {boneName=dyn_clothA_l_02}
    componentAppliedLinearAccClamp=<2000 2000 2000>
    componentLinearAccScale=<1 1 1>
    componentLinearVelScale=<1 1 1>
    gravityOverride {z=-980}
    physicsBodyDefinitions [{
      boundBone {boneName=dyn_clothA_l_02}
      boxExtents=<20 30 30>
      constraintSetup {angularLimitsMax=<0 3 0>, angularLimitsMin=<0 -3 -15>, angularTarget=<1 0 0>}
      localJointOffset=<4 0 0>
     }]
   }
   n4 = Anim Dynamics - Bone: dyn_clothA_l_03(ComponentPose=@n3, Alpha=1, ExternalForce="0,0,0") {
    angularDampingOverride=1
    angularSpringConstant=8000
    bAngularSpring=true
    bOverrideAngularDamping=true
    bUseGravityOverride=true
    gravityScale=0.35
    simulationSpace=RootRelative
    alphaBoolBlend {blendOption=Linear}
    boundBone {boneName=dyn_clothA_l_03}
    componentAppliedLinearAccClamp=<2000 2000 2000>
    componentLinearAccScale=<1 1 1>
    componentLinearVelScale=<1 1 1>
    gravityOverride {z=-980}
    physicsBodyDefinitions [{
      boundBone {boneName=dyn_clothA_l_03}
      boxExtents=<20 30 30>
      constraintSetup {angularLimitsMax=<0 3 0>, angularLimitsMin=<0 -3 -15>, angularTarget=<1 0 0>}
      localJointOffset=<4 0 0>
     }]
   }
   n5 = Anim Dynamics - Bone: dyn_clothA_l_04(ComponentPose=@n4, Alpha=1, ExternalForce="0,0,0") {
    angularDampingOverride=1
    angularSpringConstant=8000
    bAngularSpring=true
    bOverrideAngularDamping=true
    bUseGravityOverride=true
    gravityScale=0.35
    simulationSpace=RootRelative
    alphaBoolBlend {blendOption=Linear}
    boundBone {boneName=dyn_clothA_l_04}
    componentAppliedLinearAccClamp=<2000 2000 2000>
    componentLinearAccScale=<1 1 1>
    componentLinearVelScale=<1 1 1>
    gravityOverride {z=-980}
    physicsBodyDefinitions [{
      boundBone {boneName=dyn_clothA_l_04}
      boxExtents=<20 30 30>
      constraintSetup {angularLimitsMax=<0 3 0>, angularLimitsMin=<0 -3 -15>, angularTarget=<1 0 0>}
      localJointOffset=<4 0 0>
     }]
   }
   n6 = Anim Dynamics - Bone: dyn_clothA_r_03(ComponentPose=@n5, Alpha=1, ExternalForce="0,0,0") {
    angularDampingOverride=1
    angularSpringConstant=8000
    bAngularSpring=true
    bOverrideAngularDamping=true
    bUseGravityOverride=true
    gravityScale=0.35
    simulationSpace=RootRelative
    alphaBoolBlend {blendOption=Linear}
    boundBone {boneName=dyn_clothA_r_03}
    componentAppliedLinearAccClamp=<2000 2000 2000>
    componentLinearAccScale=<1 1 1>
    componentLinearVelScale=<1 1 1>
    gravityOverride {z=-980}
    physicsBodyDefinitions [{
      boundBone {boneName=dyn_clothA_r_03}
      boxExtents=<20 30 30>
      constraintSetup {angularLimitsMax=<0 3 0>, angularLimitsMin=<0 -3 -15>, angularTarget=<1 0 0>}
      localJointOffset=<4 0 0>
     }]
   }
   n7 = Anim Dynamics - Bone: dyn_clothA_r_04(ComponentPose=@n6, Alpha=1, ExternalForce="0,0,0") {
    angularDampingOverride=1
    angularSpringConstant=8000
    bAngularSpring=true
    bOverrideAngularDamping=true
    bUseGravityOverride=true
    gravityScale=0.35
    simulationSpace=RootRelative
    alphaBoolBlend {blendOption=Linear}
    boundBone {boneName=dyn_clothA_r_04}
    componentAppliedLinearAccClamp=<2000 2000 2000>
    componentLinearAccScale=<1 1 1>
    componentLinearVelScale=<1 1 1>
    gravityOverride {z=-980}
    physicsBodyDefinitions [{
      boundBone {boneName=dyn_clothA_r_04}
      boxExtents=<20 30 30>
      constraintSetup {angularLimitsMax=<0 3 0>, angularLimitsMin=<0 -3 -15>, angularTarget=<1 0 0>}
      localJointOffset=<4 0 0>
     }]
   }
   n8 = Anim Dynamics - Bone: dyn_clothB_l_02(ComponentPose=@n7, Alpha=1, ExternalForce="0,0,0") {
    angularDampingOverride=1
    angularSpringConstant=8
    bAngularSpring=true
    bOverrideAngularDamping=true
    bUseGravityOverride=true
    gravityScale=0.35
    simulationSpace=RootRelative
    alphaBoolBlend {blendOption=Linear}
    boundBone {boneName=dyn_clothB_l_02}
    componentAppliedLinearAccClamp=<2000 2000 2000>
    componentLinearAccScale=<1 1 1>
    componentLinearVelScale=<1 1 1>
    gravityOverride {z=-980}
    physicsBodyDefinitions [{
      boundBone {boneName=dyn_clothB_l_02}
      boxExtents=<20 30 30>
      constraintSetup {angularLimitsMax=<0 5 -2>, angularLimitsMin=<0 -3 -15>, angularTarget=<1 0 0>}
      localJointOffset=<4 0 0>
     }]
   }
   n9 = Anim Dynamics - Bone: dyn_clothB_l_03(ComponentPose=@n8, Alpha=1, ExternalForce="0,0,0") {
    angularDampingOverride=1
    angularSpringConstant=8
    bAngularSpring=true
    bOverrideAngularDamping=true
    bUseGravityOverride=true
    gravityScale=0.35
    simulationSpace=RootRelative
    alphaBoolBlend {blendOption=Linear}
    boundBone {boneName=dyn_clothB_l_03}
    componentAppliedLinearAccClamp=<2000 2000 2000>
    componentLinearAccScale=<1 1 1>
    componentLinearVelScale=<1 1 1>
    gravityOverride {z=-980}
    physicsBodyDefinitions [{
      boundBone {boneName=dyn_clothB_l_03}
      boxExtents=<20 30 30>
      constraintSetup {angularLimitsMax=<0 5 -2>, angularLimitsMin=<0 -3 -15>, angularTarget=<1 0 0>}
      localJointOffset=<4 0 0>
     }]
   }
   n10 = Anim Dynamics - Bone: dyn_clothB_l_04(ComponentPose=@n9, Alpha=1, ExternalForce="0,0,0") {
    angularDampingOverride=1
    angularSpringConstant=8
    bAngularSpring=true
    bOverrideAngularDamping=true
    bUseGravityOverride=true
    gravityScale=0.35
    simulationSpace=RootRelative
    alphaBoolBlend {blendOption=Linear}
    boundBone {boneName=dyn_clothB_l_04}
    componentAppliedLinearAccClamp=<2000 2000 2000>
    componentLinearAccScale=<1 1 1>
    componentLinearVelScale=<1 1 1>
    gravityOverride {z=-980}
    physicsBodyDefinitions [{
      boundBone {boneName=dyn_clothB_l_04}
      boxExtents=<20 30 30>
      constraintSetup {angularLimitsMax=<0 5 -2>, angularLimitsMin=<0 -3 -15>, angularTarget=<1 0 0>}
      localJointOffset=<4 0 0>
     }]
   }
   n11 = Anim Dynamics - Bone: dyn_clothB_r_02(ComponentPose=@n10, Alpha=1, ExternalForce="0,0,0") {
    angularDampingOverride=1
    angularSpringConstant=8
    bAngularSpring=true
    bOverrideAngularDamping=true
    bUseGravityOverride=true
    gravityScale=0.35
    simulationSpace=RootRelative
    alphaBoolBlend {blendOption=Linear}
    boundBone {boneName=dyn_clothB_r_02}
    componentAppliedLinearAccClamp=<2000 2000 2000>
    componentLinearAccScale=<1 1 1>
    componentLinearVelScale=<1 1 1>
    gravityOverride {z=-980}
    physicsBodyDefinitions [{
      boundBone {boneName=dyn_clothB_r_02}
      boxExtents=<20 30 30>
      constraintSetup {angularLimitsMax=<0 5 -2>, angularLimitsMin=<0 -3 -15>, angularTarget=<1 0 0>}
      localJointOffset=<4 0 0>
     }]
   }
   n12 = Anim Dynamics - Bone: dyn_clothB_r_03(ComponentPose=@n11, Alpha=1, ExternalForce="0,0,0") {
    angularDampingOverride=1
    angularSpringConstant=8
    bAngularSpring=true
    bOverrideAngularDamping=true
    bUseGravityOverride=true
    gravityScale=0.35
    simulationSpace=RootRelative
    alphaBoolBlend {blendOption=Linear}
    boundBone {boneName=dyn_clothB_r_03}
    componentAppliedLinearAccClamp=<2000 2000 2000>
    componentLinearAccScale=<1 1 1>
    componentLinearVelScale=<1 1 1>
    gravityOverride {z=-980}
    physicsBodyDefinitions [{
      boundBone {boneName=dyn_clothB_r_03}
      boxExtents=<20 30 30>
      constraintSetup {angularLimitsMax=<0 5 -2>, angularLimitsMin=<0 -3 -15>, angularTarget=<1 0 0>}
      localJointOffset=<4 0 0>
     }]
   }
   n13 = Anim Dynamics - Bone: dyn_clothB_r_04(ComponentPose=@n12, Alpha=1, ExternalForce="0,0,0") {
    angularDampingOverride=1
    angularSpringConstant=8
    bAngularSpring=true
    bOverrideAngularDamping=true
    bUseGravityOverride=true
    gravityScale=0.35
    simulationSpace=RootRelative
    alphaBoolBlend {blendOption=Linear}
    boundBone {boneName=dyn_clothB_r_04}
    componentAppliedLinearAccClamp=<2000 2000 2000>
    componentLinearAccScale=<1 1 1>
    componentLinearVelScale=<1 1 1>
    gravityOverride {z=-980}
    physicsBodyDefinitions [{
      boundBone {boneName=dyn_clothB_r_04}
      boxExtents=<20 30 30>
      constraintSetup {angularLimitsMax=<0 5 -2>, angularLimitsMin=<0 -3 -15>, angularTarget=<1 0 0>}
      localJointOffset=<4 0 0>
     }]
   }
   n14 = Get bIsSkydiveFloating()
   n15 = "Transform (Modify) Bone - Bone: dyn_clothC_l_MSTR"!(ComponentPose=@n13, bAlphaBoolEnabled:bool=@n14, Translation="0,-5,0", Rotation="0,0,0", Scale="1,1,1") -> (Pose) {
    alphaInputType=Bool
    rotationMode=BMM_Additive
    rotationSpace=BCS_BoneSpace
    translationMode=BMM_Additive
    translationSpace=BCS_BoneSpace
    alphaBoolBlend {blendOption=Linear}
    boneToModify {boneName=dyn_clothC_l_MSTR}
    translation {y=-5}
   }
   n16 = "Transform (Modify) Bone - Bone: dyn_clothC_r_MSTR"!(ComponentPose=@n15, bAlphaBoolEnabled:bool=@n14, Translation="0,-5,0", Rotation="0,0,0", Scale="1,1,1") -> (Pose) {
    alphaInputType=Bool
    rotationMode=BMM_Additive
    rotationSpace=BCS_BoneSpace
    translationMode=BMM_Additive
    translationSpace=BCS_BoneSpace
    alphaBoolBlend {blendOption=Linear}
    boneToModify {boneName=dyn_clothC_r_MSTR}
    translation {y=-5}
   }
   Save cached pose 'PreSimCache'(Pose=Component To Local(ComponentPose=@n16)) {cacheName=PreSimCache}
   n17 = RigidBody!(
    ComponentPose=Local To Component(LocalPose=Use cached pose 'skirtCache'())
    Alpha
    OverrideWorldGravity:Vector=Get Gravity Override()
    ExternalForce
   ) -> (Pose) {
    bClampLinearTranslationLimitToRefPose=true
    bOverrideWorldGravity=true
    overridePhysicsAsset=&/M_MED_AnglePatch_Royal_LoinCloth_Physics
    simulationSpace=BaseBoneSpace
    alphaBoolBlend {blendOption=Linear}
    baseBoneRef {boneName=pelvis}
    componentAppliedLinearAccClamp=<2000 2000 2000>
    simSpaceSettings {
     maxAngularAcceleration=1000
     maxAngularVelocity=3
     maxLinearAcceleration=250
     maxLinearVelocity=500
     velocityScaleZ=0
     worldAlpha=1
     externalLinearDragV {y=0.5}
    }
   }
   n18 = Anim Dynamics - Bone: dyn_loinCloth_l_01(ComponentPose=Local To Component(LocalPose=Use cached pose 'skirtCache'()), Alpha=1, ExternalForce="0,0,0") {
    angularDampingOverride=1
    angularSpringConstant=7000
    bAngularSpring=true
    bOverrideAngularDamping=true
    bUseGravityOverride=true
    gravityScale=0.35
    simulationSpace=RootRelative
    alphaBoolBlend {blendOption=Linear}
    boundBone {boneName=dyn_loinCloth_l_01}
    componentAppliedLinearAccClamp=<2000 2000 2000>
    componentLinearAccScale=<1 1 1>
    componentLinearVelScale=<1 1 1>
    gravityOverride {z=-980}
    physicsBodyDefinitions [{
      boundBone {boneName=dyn_loinCloth_l_01}
      boxExtents=<20 30 30>
      constraintSetup {angularLimitsMax=<15 15 -5>, angularLimitsMin=<-15 -15 -15>, angularTarget=<1 0 0>}
      localJointOffset=<4 0 0>
     }]
   }
   n19 = Anim Dynamics - Bone: dyn_loinCloth_l_02(ComponentPose=@n18, Alpha=1, ExternalForce="0,0,0") {
    angularDampingOverride=1
    angularSpringConstant=7000
    bAngularSpring=true
    bOverrideAngularDamping=true
    bUseGravityOverride=true
    gravityScale=0.35
    simulationSpace=RootRelative
    alphaBoolBlend {blendOption=Linear}
    boundBone {boneName=dyn_loinCloth_l_02}
    componentAppliedLinearAccClamp=<2000 2000 2000>
    componentLinearAccScale=<1 1 1>
    componentLinearVelScale=<1 1 1>
    gravityOverride {z=-980}
    physicsBodyDefinitions [{
      boundBone {boneName=dyn_loinCloth_l_02}
      boxExtents=<20 30 30>
      constraintSetup {angularLimitsMax=<15 15 0>, angularLimitsMin=<-15 -15 -15>, angularTarget=<1 0 0>}
      localJointOffset=<4 0 0>
     }]
   }
   n20 = Anim Dynamics - Bone: dyn_loinCloth_l_03(ComponentPose=@n19, Alpha=1, ExternalForce="0,0,0") {
    angularDampingOverride=1
    angularSpringConstant=7000
    bAngularSpring=true
    bOverrideAngularDamping=true
    bUseGravityOverride=true
    gravityScale=0.35
    simulationSpace=RootRelative
    alphaBoolBlend {blendOption=Linear}
    boundBone {boneName=dyn_loinCloth_l_03}
    componentAppliedLinearAccClamp=<2000 2000 2000>
    componentLinearAccScale=<1 1 1>
    componentLinearVelScale=<1 1 1>
    gravityOverride {z=-980}
    physicsBodyDefinitions [{
      boundBone {boneName=dyn_loinCloth_l_03}
      boxExtents=<20 30 30>
      constraintSetup {angularLimitsMax=<15 15 0>, angularLimitsMin=<-15 -15 -15>, angularTarget=<1 0 0>}
      localJointOffset=<4 0 0>
     }]
   }
   n21 = Anim Dynamics - Bone: dyn_loinCloth_l_04(ComponentPose=@n20, Alpha=1, ExternalForce="0,0,0") {
    angularDampingOverride=1
    angularSpringConstant=7000
    bAngularSpring=true
    bOverrideAngularDamping=true
    bUseGravityOverride=true
    gravityScale=0.35
    simulationSpace=RootRelative
    alphaBoolBlend {blendOption=Linear}
    boundBone {boneName=dyn_loinCloth_l_04}
    componentAppliedLinearAccClamp=<2000 2000 2000>
    componentLinearAccScale=<1 1 1>
    componentLinearVelScale=<1 1 1>
    gravityOverride {z=-980}
    physicsBodyDefinitions [{
      boundBone {boneName=dyn_loinCloth_l_04}
      boxExtents=<20 30 30>
      constraintSetup {angularLimitsMax=<15 15 0>, angularLimitsMin=<-15 -15 -15>, angularTarget=<1 0 0>}
      localJointOffset=<4 0 0>
     }]
   }
   n22 = Anim Dynamics - Bone: dyn_loinCloth_r_01(ComponentPose=@n21, Alpha=1, ExternalForce="0,0,0") {
    angularDampingOverride=1
    angularSpringConstant=7000
    bAngularSpring=true
    bOverrideAngularDamping=true
    bUseGravityOverride=true
    gravityScale=0.35
    simulationSpace=RootRelative
    alphaBoolBlend {blendOption=Linear}
    boundBone {boneName=dyn_loinCloth_r_01}
    componentAppliedLinearAccClamp=<2000 2000 2000>
    componentLinearAccScale=<1 1 1>
    componentLinearVelScale=<1 1 1>
    gravityOverride {z=-980}
    physicsBodyDefinitions [{
      boundBone {boneName=dyn_loinCloth_r_01}
      boxExtents=<20 30 30>
      constraintSetup {angularLimitsMax=<15 15 -10>, angularLimitsMin=<-15 -15 -15>, angularTarget=<1 0 0>}
      localJointOffset=<4 0 0>
     }]
   }
   n23 = Anim Dynamics - Bone: dyn_loinCloth_r_02(ComponentPose=@n22, Alpha=1, ExternalForce="0,0,0") {
    angularDampingOverride=1
    angularSpringConstant=7000
    bAngularSpring=true
    bOverrideAngularDamping=true
    bUseGravityOverride=true
    gravityScale=0.35
    simulationSpace=RootRelative
    alphaBoolBlend {blendOption=Linear}
    boundBone {boneName=dyn_loinCloth_r_02}
    componentAppliedLinearAccClamp=<2000 2000 2000>
    componentLinearAccScale=<1 1 1>
    componentLinearVelScale=<1 1 1>
    gravityOverride {z=-980}
    physicsBodyDefinitions [{
      boundBone {boneName=dyn_loinCloth_r_02}
      boxExtents=<20 30 30>
      constraintSetup {angularLimitsMax=<15 15 0>, angularLimitsMin=<-15 -15 -15>, angularTarget=<1 0 0>}
      localJointOffset=<4 0 0>
     }]
   }
   n24 = Anim Dynamics - Bone: dyn_loinCloth_r_03(ComponentPose=@n23, Alpha=1, ExternalForce="0,0,0") {
    angularDampingOverride=1
    angularSpringConstant=7000
    bAngularSpring=true
    bOverrideAngularDamping=true
    bUseGravityOverride=true
    gravityScale=0.35
    simulationSpace=RootRelative
    alphaBoolBlend {blendOption=Linear}
    boundBone {boneName=dyn_loinCloth_r_03}
    componentAppliedLinearAccClamp=<2000 2000 2000>
    componentLinearAccScale=<1 1 1>
    componentLinearVelScale=<1 1 1>
    gravityOverride {z=-980}
    physicsBodyDefinitions [{
      boundBone {boneName=dyn_loinCloth_r_03}
      boxExtents=<20 30 30>
      constraintSetup {angularLimitsMax=<15 15 0>, angularLimitsMin=<-15 -15 -15>, angularTarget=<1 0 0>}
      localJointOffset=<4 0 0>
     }]
   }
   n25 = Anim Dynamics - Bone: dyn_loinCloth_r_04(ComponentPose=@n24, Alpha=1, ExternalForce="0,0,0") {
    angularDampingOverride=1
    angularSpringConstant=7000
    bAngularSpring=true
    bOverrideAngularDamping=true
    bUseGravityOverride=true
    gravityScale=0.35
    simulationSpace=RootRelative
    alphaBoolBlend {blendOption=Linear}
    boundBone {boneName=dyn_loinCloth_r_04}
    componentAppliedLinearAccClamp=<2000 2000 2000>
    componentLinearAccScale=<1 1 1>
    componentLinearVelScale=<1 1 1>
    gravityOverride {z=-980}
    physicsBodyDefinitions [{
      boundBone {boneName=dyn_loinCloth_r_04}
      boxExtents=<20 30 30>
      constraintSetup {angularLimitsMax=<15 15 0>, angularLimitsMin=<-15 -15 -15>, angularTarget=<1 0 0>}
      localJointOffset=<4 0 0>
     }]
   }
   n26 = "Transform (Modify) Bone - Bone: dyn_loinCloth_l_MSTR"(ComponentPose=@n25, Alpha=1, Translation="0,-1,0", Rotation="0,0,0", Scale="1,1,1") {
    translationMode=BMM_Additive
    translationSpace=BCS_BoneSpace
    alphaBoolBlend {blendOption=Linear}
    boneToModify {boneName=dyn_loinCloth_l_MSTR}
    translation {y=-1}
   }
   n27 = "Transform (Modify) Bone - Bone: dyn_loinCloth_r_MSTR"(ComponentPose=@n26, Alpha=1, Translation="0,-1,0", Rotation="0,0,0", Scale="1,1,1") {
    translationMode=BMM_Additive
    translationSpace=BCS_BoneSpace
    alphaBoolBlend {blendOption=Linear}
    boneToModify {boneName=dyn_loinCloth_r_MSTR}
    translation {y=-1}
   }
   n28 = "Transform (Modify) Bone - Bone: dyn_clothA_l_MSTR"(ComponentPose=@n27, Alpha=1, Translation="0,2,0", Rotation="0,-7,0", Scale="1,1,1") {
    rotationMode=BMM_Additive
    rotationSpace=BCS_BoneSpace
    translationMode=BMM_Additive
    translationSpace=BCS_BoneSpace
    alphaBoolBlend {blendOption=Linear}
    boneToModify {boneName=dyn_clothA_l_MSTR}
    rotation {yaw=-7}
    translation {y=2}
   }
   n29 = "Transform (Modify) Bone - Bone: dyn_clothA_r_MSTR"(ComponentPose=@n28, Alpha=1, Translation="0,2,0", Rotation="0,-8,0", Scale="1,1,1") {
    rotationMode=BMM_Additive
    rotationSpace=BCS_BoneSpace
    translationMode=BMM_Additive
    translationSpace=BCS_BoneSpace
    alphaBoolBlend {blendOption=Linear}
    boneToModify {boneName=dyn_clothA_r_MSTR}
    rotation {yaw=-8}
    translation {y=2}
   }
   Save cached pose 'RBANCache'(Pose=Blend Poses by bool(
     BlendPose_0=Component To Local(ComponentPose=@n29)
     BlendPose_1=Component To Local(ComponentPose=@n17)
     BlendTime_0=0.3
     BlendTime_1=0.3
     bActiveValue=Get isFrontEndNotEmoting() -> (isFrontEndNotEmoting:bool)
    ) {blendType=Linear}) {cacheName=RBANCache}
   n30 = Get bIsSprinting()
   n31 = "Transform (Modify) Bone - Bone: dyn_clothB_l_MSTR"!(
    ComponentPose=Local To Component(LocalPose=Use cached pose 'RBANCache'())
    bAlphaBoolEnabled:bool=@n30
    Translation="0,0,0"
    Rotation="0,10,0"
    Scale="1,1,1"
   ) -> (Pose) {
    alphaInputType=Bool
    rotationMode=BMM_Additive
    rotationSpace=BCS_ParentBoneSpace
    alphaBoolBlend {blendInTime=1, blendOption=Linear, blendOutTime=1}
    boneToModify {boneName=dyn_clothB_l_MSTR}
    rotation {yaw=10}
   }
   n32 = "Transform (Modify) Bone - Bone: dyn_clothB_r_MSTR"!(ComponentPose=@n31, bAlphaBoolEnabled:bool=@n30, Translation="0,0,0", Rotation="0,10,0", Scale="1,1,1") -> (Pose) {
    alphaInputType=Bool
    rotationMode=BMM_Additive
    rotationSpace=BCS_ParentBoneSpace
    alphaBoolBlend {blendInTime=1, blendOption=Linear, blendOutTime=1}
    boneToModify {boneName=dyn_clothB_r_MSTR}
    rotation {yaw=10}
   }
   n33 = Get isCustomCape() -> (isCustomCape:bool)
   n34 = "Transform (Modify) Bone - Bone: dyn_shoulderPadBack_l_01"!(ComponentPose=@n32, bAlphaBoolEnabled:bool=@n33, Translation="0,-1.5,0", Rotation="0,0,0", Scale="1,1,1") -> (Pose) {
    alphaInputType=Bool
    rotationMode=BMM_Additive
    rotationSpace=BCS_BoneSpace
    translationMode=BMM_Additive
    translationSpace=BCS_BoneSpace
    alphaBoolBlend {blendOption=Linear}
    boneToModify {boneName=dyn_shoulderPadBack_l_01}
    translation {y=-1.5}
   }
   n35 = "Transform (Modify) Bone - Bone: dyn_shoulderPadBack_r_01"!(ComponentPose=@n34, bAlphaBoolEnabled:bool=@n33, Translation="0,-1.5,0", Rotation="0,0,0", Scale="1,1,1") -> (Pose) {
    alphaInputType=Bool
    rotationMode=BMM_Additive
    rotationSpace=BCS_BoneSpace
    translationMode=BMM_Additive
    translationSpace=BCS_BoneSpace
    alphaBoolBlend {blendOption=Linear}
    boneToModify {boneName=dyn_shoulderPadBack_r_01}
    translation {y=-1.5}
   }
   n36 = "Transform (Modify) Bone - Bone: dyn_loinCloth_l_MSTR"!(
    ComponentPose=@n35
    bAlphaBoolEnabled:bool=Get isPlayingHightowerEmote() -> (isPlayingHightowerEmote:bool)
    Translation="0,0,0"
    Rotation="5,-10,0"
    Scale="1,1,1"
   ) -> (Pose) {
    alphaInputType=Bool
    rotationMode=BMM_Additive
    rotationSpace=BCS_BoneSpace
    alphaBoolBlend {blendInTime=1, blendOption=Linear, blendOutTime=1}
    boneToModify {boneName=dyn_loinCloth_l_MSTR}
    rotation {pitch=5, yaw=-10}
   }
   Save cached pose 'PostSimCache'(Pose=Component To Local(ComponentPose=@n36)) {cacheName=PostSimCache}
   n37 = RigidBody(ComponentPose=Local To Component(LocalPose=Use cached pose 'PreSimCache'()), OverrideWorldGravity=Get Gravity Override()) {
    bClampLinearTranslationLimitToRefPose=true
    bOverrideWorldGravity=true
    overridePhysicsAsset=&/M_MED_AnglePatch_Royal_Skirt_Physics
    simulationSpace=BaseBoneSpace
    alphaBoolBlend {blendOption=Linear}
    baseBoneRef {boneName=spine_01}
    componentAppliedLinearAccClamp=<2000 2000 2000>
    simSpaceSettings {
     maxAngularAcceleration=1000
     maxAngularVelocity=3
     maxLinearAcceleration=250
     maxLinearVelocity=500
     velocityScaleZ=0
     worldAlpha=1
     externalLinearDragV {y=0.5}
    }
   }
   Save cached pose 'skirtCache'(Pose=Component To Local(ComponentPose=@n37)) {cacheName=skirtCache}
  ]}
 EventGraph {nodes [
   n1 = Event BlueprintUpdateAnimation()
   > Set Delta Time X(Delta Time X=@n1.DeltaTimeX)
   > n2 = Sequence!(execute) -> (then_0, then_1, then_2:Exec, then_3:Exec, then_4:Exec, then_5:Exec, then_6:Exec, then_7:Exec)
   n3 = PlayerMovement(execute=@n2.then_2, AnimInstance=Self-Reference(), XDotProductCoefficient=1, YDotProductCoefficient=1, ZDotProductCoefficient=1)
   > n4 = Sequence!(execute) -> (then_0, then_1, then_2)
   > n5 = Branch(Condition=OR Boolean!(A="Not Equal (Float)"(A=@n3), B="Not Equal (Float)"(A=@n3.YMovement), C:bool="Not Equal (Float)"(A=@n3.ZMovement)) -> (ReturnValue))
   > Set isWalking(isWalking:bool=true) -> (Output_Get:bool)
   Set isWalking(execute=@n5.else, isWalking=false)
   n6 = "float > float"(A=MapRangeClamped(Value=@n3, InRangeA=-1, InRangeB=0, OutRangeA=1, OutRangeB=0), B=0.5)
   n7 = MapRangeClamped(Value=@n3, InRangeA=0, InRangeB=1)
   n8 = Branch(execute=@n4.then_1, Condition=OR Boolean!(A="float > float"(A=@n7, B=0.5), B=@n6) -> (ReturnValue))
   > Set PlayerStrafe(PlayerStrafe:bool=true) -> (Output_Get:bool)
   Set PlayerStrafe(execute=@n8.else, PlayerStrafe=false)
   n9 = Branch(execute=@n4.then_2, Condition=AND Boolean(A=@n6, B=Get bIsCrouching()))
   > Set isCrouchStrafeLeft(isCrouchStrafeLeft:bool=true) -> (Output_Get:bool)
   Set isCrouchStrafeLeft(execute=@n9.else, isCrouchStrafeLeft=false)
   n10 = Branch(execute=@n2.then_7, Condition=Get bIsCrouching())
   > Set isCrouching(isCrouching:double=1) -> (Output_Get:double)
   Set isCrouching(execute=@n10.else, isCrouching=0)
   n11 = Gravity Override State Function DEPRECATED(
    execute=@n2.then_1
    CharacterPartAnimInstance=Self-Reference()
    GameTimeInSeconds=Get Delta Time X()
    GravityOverrideParams=Get Gravity Override Params()
   )
   > Set Gravity Override(Gravity Override=@n11)
   n12 = Branch(execute=@n2.then_3, Condition=AND Boolean(A=Get bIsCrouching(), B=Get isWalking() -> (isWalking:bool)))
   > Set isCrouchWalking(isCrouchWalking:bool=true) -> (Output_Get:bool)
   > Set isCrouchWalkingAlpha(isCrouchWalkingAlpha:double=0) -> (Output_Get:double)
   Set isCrouchWalking(execute=@n12.else, isCrouchWalking=false)
   > Set isCrouchWalkingAlpha(isCrouchWalkingAlpha=1)
   n13 = Get bIsPlayingEmote()
   n14 = Get isPlayingHightowerEmote()
   n15 = Branch(execute=@n2.then_4, Condition=AND Boolean!(A=Get bIsInFrontEnd(), B=NOT Boolean(A=@n13), C:bool=NOT Boolean(A=@n14)) -> (ReturnValue))
   > Set isFrontEndNotEmoting(isFrontEndNotEmoting:bool=true) -> (Output_Get:bool)
   Set isFrontEndNotEmoting(execute=@n15.else, isFrontEndNotEmoting=false)
   n16 = Cast To FortPlayerPawn(execute=@n2.then_6, Object=GetOwningActor())
   > n17 = Set Glider(Glider:Athena Glider Item Definition Object Reference=GetItemDefinitionTypeFromLoadoutWithActor!(
     Actor:Actor Object Reference=@n16
     LoadoutSlot:Gameplay Tag Structure=(TagName="LoadoutSlot.Glider")
     CosmeticType:Athena Cosmetic Item Definition Class Reference=/Script/FortniteGame.AthenaGliderItemDefinition
    ) -> (OutCosmeticItemDef:Athena Glider Item Definition Object Reference)) -> (Output_Get:Athena Glider Item Definition Object Reference)
   > n18 = Branch(Condition=AND Boolean!(
     A="Equal (Object)"(A=@n17, B=/BRCosmetics/Athena/Items/Cosmetics/Gliders/Glider_ID_232_HightowerDate) {is=title K2Node_CallFunction}
     B=Get bIsParachuteOpened()
    ) -> (ReturnValue))
   > Set isPlayingHightowerEmote(isPlayingHightowerEmote:bool=true) -> (Output_Get:bool)
   Set isPlayingHightowerEmote(execute=@n18.else, isPlayingHightowerEmote=false)
   n19 = Branch(
    execute=@n2.then_5
    Condition=BackpackItemDefinitionHasAnyMetaTag!(Tags:Array of Gameplay Tag Structures=Make Array("[0]"=(TagName="Cosmetics.MetaTags.CapeOn")), bRequiresEquipped:bool=false) -> (ReturnValue:bool)
   )
   > Set isCustomCape(isCustomCape:bool=true) -> (Output_Get:bool)
   Set isCustomCape(execute=@n19.else, isCustomCape=false)
   Montage_IsPlaying(
    self=GetAnimInstance(self=GetOwnerSkeletalMeshComponent())
    Montage=/BRCosmetics/Animation/Game/MainPlayer/Emotes/HighTowerDate/Emote_HighTowerDate_Hero_M
   )
   GetAnimInstance!(Context:"Anim Execution Context Structure (by ref)") -> (ReturnValue)
   Get PlayerGliderType() -> (PlayerGliderType=HangGlider)
  ]}
]
variables [
 BodyMesh {type=FortPlayerAnimInstance, default=null}
 Delta Time X {type=number}
 Glider {type=AthenaGliderItemDefinition, default=null}
 Gravity Override {type=Vector}
 Gravity Override Params {type=GravityOverrideParamsStruct}
 isCrouching {type=number}
 isCrouchStrafeLeft {type=boolean}
 isCrouchWalking {type=boolean}
 isCrouchWalkingAlpha {type=number}
 isCustomCape {type=boolean}
 isFrontEndNotEmoting {type=boolean}
 isPlayingHightowerEmote {type=boolean}
 isWalking {type=boolean}
 NewVar {type=EFortCustomPartType, default=Head}
 PlayerStrafe {type=boolean}
 Target {type=FortPlayerPawn}
]
