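# Physics asset supplying cape collision shapes for the M_MED_AnglePatch_Royal male player cosmetic body.
# Capsules wrap the calves, pelvis, spine, clavicles, neck, right forearm and four dyn_clothC bones, with a sphere on the head.
# Bodies for the feet, thighs, left forearm and right upper arm carry no shapes.
# It defines no constraints, so it serves as a collider set rather than a jointed ragdoll.
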PhysicsAsset /BRCosmetics/Characters/Player/Male/Medium/Bodies/M_MED_AnglePatch_Royal/Meshes/M_MED_AnglePatch_Royal_CapeColliders
bodies {
 calf_l [{shapeType=Capsule, length=40.5092, radius=10.7509, center=<-24.2345 -0.8139 -0.373>, rotation=<rot 90 179.7762 -177.4742>}]
 calf_r [{shapeType=Capsule, length=40.5092, radius=10.7509, center=<24.2345 0.8139 -0.373>, rotation=<rot 90 179.7762 -177.4742>}]
 clavicle_l [{shapeType=Capsule, length=9.0192, radius=9.8871, center=<9.4869 -1.9815 -2.5128>, rotation=<rot 69.3297 1.4476 176.9462>}]
 clavicle_r [{shapeType=Capsule, length=9.0192, radius=9.8871, center=<-9.4869 1.9815 2.5128>, rotation=<rot 69.3297 1.4476 -3.0538>}]
 dyn_clothC_l_01 [{
   shapeType=Capsule
   length=13
   radius=10.3038
   shapeName=dyn_clothC_l_01_capsule
   center=<8.3746 7.8931 0>
   rotation=<rot 90 -161.662 -162.3713>
  }]
 dyn_clothC_l_03 [{shapeType=Capsule, length=15.5, radius=10, shapeName=dyn_clothC_l_03_capsule, center=<6.2168 5.0582 0>, rotation=<rot 90 0 0>}]
 dyn_clothC_r_01 [{shapeType=Capsule, length=13, radius=10.3038, shapeName=dyn_clothC_r_01_capsule, center=<8.3746 7.8931 0>, rotation=<rot 90 51.1528 50.4435>}]
 dyn_clothC_r_03 [{shapeType=Capsule, length=15.5, radius=10, shapeName=dyn_clothC_r_03_capsule, center=<6.2168 5.0582 0>, rotation=<rot 90 0 0>}]
 foot_l []
 foot_r []
 head [{shapeType=Sphere, radius=17.3125, center=<11.028 -0.174 0>}]
 lowerarm_l []
 lowerarm_r [{
   shapeType=Capsule
   length=11.1695
   radius=5.6002
   shapeName=lowerarm_r_capsule
   center=<-1.2754 -1.0204 0.6319>
   rotation=<rot 75.186 169.8601 99.2883>
  }]
 neck_02 [{shapeType=Capsule, length=13.4375, radius=7.3175, center=<-0.2642 -0.3408 0>, rotation=<rot 90 -171.1193 -175.3978>}]
 pelvis [{shapeType=Capsule, length=8.6456, radius=14.5388, center=<-2.4887 0.1173 0>}]
 spine_01 [{shapeType=Capsule, length=7.9438, radius=13.5586, center=<-5.316 -0.2934 0>}]
 spine_02 [
  {shapeType=Capsule, length=5.7999, radius=14.1068, center=<-1.711 0.7644 0.0001>}
  {shapeType=Capsule, length=29.3956, radius=11.5807, center=<2.5024 0.4992 3.9825>, rotation=<rot 90 0.0049 0.0049>}
  {shapeType=Capsule, length=29.875, radius=11.1013, center=<2.4704 0.4991 -3.9825>, rotation=<rot 90 -0.0396 -0.0396>}
 ]
 spine_03 [{shapeType=Capsule, length=6.6842, radius=14.1776, center=<2.7519 1.4418 0.0001>}]
 spine_04 [{shapeType=Capsule, length=7.9139, radius=14.5355, center=<4.9925 0.1941 0.0002>}]
 spine_05 [{shapeType=Capsule, length=11.7431, radius=15.0129, center=<-0.4501 0.8394 0>}]
 thigh_l []
 thigh_r []
 upperarm_r []
}
constraints []
